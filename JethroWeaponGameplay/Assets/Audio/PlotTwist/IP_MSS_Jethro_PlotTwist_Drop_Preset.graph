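# Mono one-shot MetaSound source that plays the drop sound for the Jethro PlotTwist blaster weapon.
# On play, a wave player fires the drop wave with a pitch shift randomized within base plus or minus spread.
# A matching random gain passes through a Max node and multiplies the audio, though zero default spread keeps volume fixed.
# It uses weapon-drop attenuation and the loot-collect sound class, and signals finished when the wave ends.
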MetaSoundSource /JethroWeaponGameplay/Assets/Audio/PlotTwist/IP_MSS_Jethro_PlotTwist_Drop_Preset
inputs [
 Random.Pitch.Base {type=Float}
 Random.Pitch.Spread {type=Float}
 Random.Volume.Base {type=Float}
 Random.Volume.Spread {type=Float}
 Sound {type=WaveAsset}
 UE.Source.OnPlay {type=Trigger}
]
interfaces [UE.OutputFormat.Mono:1.0, UE.Source.OneShot:1.0, UE.Source:1.1]
outputs [
 "UE.OutputFormat.Mono.Audio:0" {type=Audio}
 UE.Source.OneShot.OnFinished {type=Trigger}
]
pages [Default {
  input_defaults {
   Random.Pitch.Base=0
   Random.Pitch.Spread=1
   Random.Volume.Base=0.597
   Random.Volume.Spread=0
   Sound=/Galileo/Items/BlasterPlotTwist/Audio/BlasterPlotTwist_Drop
   UE.Source.OnPlay=null
  }
  nodes [
   n1 = UE.Add:Float(PrimaryOperand=^Random.Volume.Base, AdditionalOperands=^Random.Volume.Spread)
   n2 = UE.Reroute(Value=^UE.Source.OnPlay)
   n3 = UE.Subtract:Float(PrimaryOperand=^Random.Volume.Base, AdditionalOperands=^Random.Volume.Spread)
   WavePlayer_0 = UE.Wave Player:Mono(
    Wave Asset=^Sound
    Play=^UE.Source.OnPlay
    Pitch Shift=UE.RandomFloat(
     Max=UE.Add:Float(PrimaryOperand=^Random.Pitch.Base, AdditionalOperands=^Random.Pitch.Spread).Out
     Next=UE.Reroute(Value=^UE.Source.OnPlay).Value
     Min=UE.Subtract:Float(PrimaryOperand=^Random.Pitch.Base, AdditionalOperands=^Random.Pitch.Spread).Out
    ).Value
   )
   Multiply_0 = UE.Multiply:Audio by Float(
    AdditionalOperands=Max.Max:Float(A=UE.RandomFloat(Max=@n1.Out, Next=@n2.Value, Min=@n3.Out).Value).Value
    PrimaryOperand=@WavePlayer_0.Out Mono
   )
  ]
  links [{dst="@UE.OutputFormat.Mono.Audio:0", src=Multiply_0.Out}, {dst="@UE.Source.OneShot.OnFinished", src=WavePlayer_0.On Finished}]
 }]
properties {
 assetImportData=&self:AssetImportData
 attenuationSettings=&Fort_Weapon_Drop_Att
 maxDistance=3000
 sampleRate=48000
 soundClassObject=&Fort_Loot_Collect
}
